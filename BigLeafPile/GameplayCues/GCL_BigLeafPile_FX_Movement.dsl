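; Looping gameplay cue that shows a camera lens effect on the player while moving in a big leaf pile.
; On loop start it adds the lens effect to the target pawn's camera manager and caches it with its primary particle component.
; On removal it sets particle parameters to animate out, waits briefly, then removes the lens effect.
; The cached particle component deactivates independently of the controller and camera validity checks.
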(event EventOnLoopingStart
  (params (Target :type Actor%20Object%20Reference) (Parameters :type Gameplay%20Cue%20Parameters%20Structure%20%28by%20ref%29) (SpawnResults :type Gameplay%20Cue%20Notify%20Spawn%20Result%20Structure%20%28by%20ref%29))
  ;[0.01782 1 0] EventOnLoopingStart
  ; Activate a cache camera lens effect for moving in a leaf pile
  (bind _asfort_player_pawn_athena (Utilities|Casting|CastToFortPlayerPawnAthena Target))
  (Utilities|IsValid (Fortnite|GetFortPlayerControllerfromActor _asfort_player_pawn_athena)
    (:Is%20Valid
      (Utilities|IsValid (Variables|/Script/Engine.PlayerController|GetPlayerCameraManager (Fortnite|GetFortPlayerControllerfromActor _asfort_player_pawn_athena))
        (:Is%20Valid
          ;;[0.4179 0 1] Camera Lens Effect
          (bind _returnvalue (Camera|AddGenericCameraLensEffect (Variables|/Script/Engine.PlayerController|GetPlayerCameraManager (Fortnite|GetFortPlayerControllerfromActor _asfort_player_pawn_athena)) "/Game/Effects/Fort_Effects/Camera_Effects/BP_Camera_Lens_BigLeafPile.BP_Camera_Lens_BigLeafPile_C")) ; FORT-584490 was accessing none here, \nvalidate before continuing
          (Variables|Default|SetFXCameraBP _returnvalue)
          (Utilities|IsValid (CameraLensEffect|GetPrimaryParticleComponent _returnvalue)
            (:Is%20Valid
              (Variables|Default|SetFXRef (CameraLensEffect|GetPrimaryParticleComponent _returnvalue)))))))))

(event EventOnRemoval
  (params (Target :type Actor%20Object%20Reference) (Parameters :type Gameplay%20Cue%20Parameters%20Structure%20%28by%20ref%29) (SpawnResults :type Gameplay%20Cue%20Notify%20Spawn%20Result%20Structure%20%28by%20ref%29))
  (bind _fxref (Variables|Default|GetFXRef))
  (bind _fxref_2 (Variables|Default|GetFXRef))
  ;[1 0.4014 0] EventOnRemoval
  ;
  (Utilities|FlowControl|Sequence
    (:then_0
      ;;[0.5 0 1] Visually  remove the VFX by setting appropriate parameters to animate out
      (Utilities|IsValid _fxref
        (:Is%20Valid
          (Effects|Components|ParticleSystem|SetFloatParameter _fxref "Exit" 0.0)
          (Effects|Components|ParticleSystem|SetBoolParameter _fxref "IsRemoving" true))))
    (:then_1
      ;;[1 0.4014 0] Remove the lens effect from the Camera
      (Utilities|FlowControl|Delay 0 1.0) ; delay to time out the FX
      (Utilities|FlowControl|Sequence
        (:then_0
          (bind _asfort_player_pawn_athena (Utilities|Casting|CastToFortPlayerPawnAthena Target))
          (Utilities|IsValid (Fortnite|GetFortPlayerControllerfromActor _asfort_player_pawn_athena)
            (:Is%20Valid
              (Utilities|IsValid (Variables|/Script/Engine.PlayerController|GetPlayerCameraManager (Fortnite|GetFortPlayerControllerfromActor _asfort_player_pawn_athena))
                (:Is%20Valid
                  (Camera|RemoveGenericCameraLensEffect (Variables|/Script/Engine.PlayerController|GetPlayerCameraManager (Fortnite|GetFortPlayerControllerfromActor _asfort_player_pawn_athena)) (Variables|Default|GetFXCameraBP)))))))
        (:then_1
          ;;[1 0.4014 0] Deactivating the stored FX is not dependent on the other validation checks since it is cached
          (Utilities|IsValid _fxref_2
            (:Is%20Valid
              (Components|Activation|Deactivate _fxref_2))))) ; second sequence here because \nlater exec line delays aren't respected
    )) ; running as a sequence because these checks\n are not dependent on each other
)

(fn ConstructionScript)
