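; Looping gameplay cue that plays bush-style enter, step and exit sounds for a player pawn in a big leaf pile.
; OnBurst adds the BigLeafPile sound library to the pawn and plays Enter; OnCeaseRelevant plays Exit and removes the library.
; OnBecomeRelevant stores the pawn for the On Player Step event, which picks a step sound by local control and scales it with movement speed.
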(event Custom|On%20Player%20Step
  ;[0.2688 0.3094 0] On player step, play this audio
  (bind _asfort_player_pawn_athena (Utilities|Casting|CastToFortPlayerPawnAthena (Variables|Default|GetTargetPawn)))
  (SoundLibrary|PlaySoundSimple (WorldSubsystems|GetSoundLibrarySubsystem _asfort_player_pawn_athena) _asfort_player_pawn_athena (Utilities|Struct|MakeSimpleSoundSettings (select (Pawn|IsLocallyControlled _asfort_player_pawn_athena) "(TagName=\"SoundLibrary.Player.State.Bush.Step.1P\")" (Variables|Default|GetFalse)) false false true false (Math|Float|MapRangeClamped (Math|Vector|VectorLength (Transformation|GetVelocity _asfort_player_pawn_athena)) 25.0 6e+02 0.5 1.3) 1.0 0.0 0 0 "None" 0 "KeepRelativeOffset" (Transformation|GetActorLocation _asfort_player_pawn_athena)) :!class /Script/SoundLibrary.SoundLibrarySubsystem))

(fn OnBurst
  (params (MyTarget :type Actor%20Object%20Reference) (Parameters :type Gameplay%20Cue%20Parameters%20Structure%20%28by%20ref%29))
  (bind _asfort_player_pawn_athena (Utilities|Casting|CastToFortPlayerPawnAthena MyTarget))
  (SoundLibrary|AddLibrary (WorldSubsystems|GetSoundLibrarySubsystem _asfort_player_pawn_athena) _asfort_player_pawn_athena "/BigLeafPile/Environmental/Audio/BigLeafPile_Lib.BigLeafPile_Lib" :!class /Script/SoundLibrary.SoundLibrarySubsystem)
  (SoundLibrary|PlaySoundSimple (WorldSubsystems|GetSoundLibrarySubsystem _asfort_player_pawn_athena) _asfort_player_pawn_athena (Utilities|Struct|MakeSimpleSoundSettings "(TagName=\"SoundLibrary.Player.State.Bush.Enter\")") :!class /Script/SoundLibrary.SoundLibrarySubsystem)
  (return false))

(fn OnCeaseRelevant
  (params (MyTarget :type Actor%20Object%20Reference) (Parameters :type Gameplay%20Cue%20Parameters%20Structure%20%28by%20ref%29))
  (Utilities|IsValid MyTarget
    (:Is%20Valid
      (bind _asfort_player_pawn_athena (Utilities|Casting|CastToFortPlayerPawnAthena MyTarget))
      (SoundLibrary|PlaySoundSimple (WorldSubsystems|GetSoundLibrarySubsystem _asfort_player_pawn_athena) _asfort_player_pawn_athena (Utilities|Struct|MakeSimpleSoundSettings "(TagName=\"SoundLibrary.Player.State.Bush.Exit\")") :!class /Script/SoundLibrary.SoundLibrarySubsystem)
      (SoundLibrary|RemoveLibrary (WorldSubsystems|GetSoundLibrarySubsystem _asfort_player_pawn_athena) _asfort_player_pawn_athena "/BigLeafPile/Environmental/Audio/BigLeafPile_Lib.BigLeafPile_Lib" :!class /Script/SoundLibrary.SoundLibrarySubsystem)
      (return false))))

(fn ConstructionScript)

(fn OnBecomeRelevant
  (params (MyTarget :type Actor%20Object%20Reference) (Parameters :type Gameplay%20Cue%20Parameters%20Structure%20%28by%20ref%29))
  (bind _asfort_player_pawn (Utilities|Casting|CastToFortPlayerPawn MyTarget))
  (Variables|Default|SetTargetPawn _asfort_player_pawn)
  (return false))
